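# Material function, exposed in the library under Time, that turns a time input into a normalized, eased progress value over a start time and duration.
# Time minus Start Time is divided by Duration and saturated, then Direction blends it with its inverse for reverse playback.
# The result samples a curve atlas row whose R, G, B and A channels hold linear, ease-in, ease-out and S-curve shapes, selected by a Curve Type switch.
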MaterialFunction /BRCosmetics/Characters/Player/Male/Medium/Bodies/M_MED_BriskReign_Glacier/Materials/Reactivity/MF_TimeCurve
graph {
 properties {bExposeToLibrary=true, libraryCategoriesText [Time]}
 nodes [
  n1 = FunctionInput() {
   bUsePreviewValueAsDefault=true
   description="0: Linear; 1: Ease In; 2: Ease Out; 3: S Curve"
   inputName=Curve Type
   inputType=FunctionInput_Scalar
   sortPriority=40
  }
  n2 = FunctionInput() {bUsePreviewValueAsDefault=true, description=in seconds, inputName=Start Time, inputType=FunctionInput_Scalar, sortPriority=10}
  n3 = Subtract(
   A=FunctionInput(Preview=Time() {period=3600}) {bUsePreviewValueAsDefault=true, description=Default : Time node, inputName=Time, inputType=FunctionInput_Scalar}
   B=@n2
  ) {constB=7200}
  n4 = FunctionInput() {
   bUsePreviewValueAsDefault=true
   description=in seconds
   inputName=Duration
   inputType=FunctionInput_Scalar
   sortPriority=20
   previewValue=<1 0 0 1>
  }
  n5 = NamedRerouteDeclaration(None=Saturate(None=Divide(A=@n3, B=@n4))) {name=Start Time, nodeColor=<rgba 0 0.7566 1 1>}
  n6 = FunctionInput() {bUsePreviewValueAsDefault=true, description="0: Forward; 1: Reverse", inputName=Direction, inputType=FunctionInput_Scalar, sortPriority=30}
  MaterialExpressionCurveAtlasRowParameter_0 = CurveAtlasRowParameter(CurveTime=LinearInterpolate(A=@n5, B=OneMinus(None=@n5), Alpha=@n6)) {
   atlas=&/CVA_Time
   curve=&/CA_TimeCurves
   parameterName=Time Curves
  }
  n7 = Switch!(
   SwitchValue=@n1
   Default
   Linear=@MaterialExpressionCurveAtlasRowParameter_0.R
   EaseIn=@MaterialExpressionCurveAtlasRowParameter_0.R
   EaseOut=@MaterialExpressionCurveAtlasRowParameter_0.R
   SCurve=@MaterialExpressionCurveAtlasRowParameter_0.R
  ) {inputs [
    {
     inputName=Linear
     input {
      expression=&self:MaterialExpressionCurveAtlasRowParameter_0
      inputName=None
      mask=1
      maskA=0
      maskB=0
      maskG=0
      maskR=1
      outputIndex=1
     }
    }
    {
     inputName=EaseIn
     input {
      expression=&self:MaterialExpressionCurveAtlasRowParameter_0
      inputName=None
      mask=1
      maskA=0
      maskB=0
      maskG=1
      maskR=0
      outputIndex=2
     }
    }
    {
     inputName=EaseOut
     input {
      expression=&self:MaterialExpressionCurveAtlasRowParameter_0
      inputName=None
      mask=1
      maskA=0
      maskB=1
      maskG=0
      maskR=0
      outputIndex=3
     }
    }
    {
     inputName=SCurve
     input {
      expression=&self:MaterialExpressionCurveAtlasRowParameter_0
      inputName=None
      mask=1
      maskA=1
      maskB=0
      maskG=0
      maskR=0
      outputIndex=4
     }
    }
   ]}
  FunctionOutput(None=@n7) {outputName=Time}
 ]
}
